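# Physics asset for the Treasure Hunter Fashion Mint backpack cosmetic, defining collision and joints for its dangling medals, map and flashlight.
# A five-link medal chain ending in a joint1 box, a map base with top and bottom branches, and the flashlight all hang off the pack body, itself constrained to root.
# Pack-to-root and map-base-to-pack joints rotate freely; the others have tight angular limits.
# Small bodies are mostly capsules, while root carries a large box and sphere.
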PhysicsAsset /BRCosmetics/Accessories/FORT_Backpacks/F_MED_Treasure_Hunter_Fashion_Mint/Meshes/F_MED_Treasure_Hunter_Fashion_Mint_Pack_Physics
bodies {
 dyn_flashlight [
  {shapeType=Capsule, length=2, radius=2, shapeName=dyn_map_bot_capsule, rotation=<rot 90 0 0>}
  {shapeType=Capsule, length=2.75, radius=1.75, shapeName=dyn_flashlight_capsule, center=<-8.3274 0 0.1729>, rotation=<rot 88.8109 0 0>}
 ]
 dyn_map_base [{
   shapeType=Capsule
   length=2.3292
   radius=2.1887
   shapeName=dyn_map_base_capsule
   center=<0.1565 -0.0886 0.0982>
   rotation=<rot 0 108.7834 -77.8234>
  }]
 dyn_map_bot [{shapeType=Capsule, length=2, radius=2, shapeName=dyn_map_bot_capsule, rotation=<rot 90 0 0>}]
 dyn_map_bot_end [{shapeType=Capsule, length=2, radius=2, shapeName=dyn_map_bot_end_capsule, rotation=<rot 90 0 0>}]
 dyn_map_top [{shapeType=Capsule, length=2, radius=2, shapeName=dyn_map_top_capsule, rotation=<rot 90 0 0>}]
 dyn_map_top_end [{shapeType=Capsule, length=2, radius=2, shapeName=dyn_map_top_end_capsule, rotation=<rot 90 0 0>}]
 dyn_medal1 [{shapeType=Capsule, length=0.255, radius=1.255, shapeName=dyn_medal1_capsule, center=<0.1867 -0.0099 -0.0292>, rotation=<rot 90 0 0>}]
 dyn_medal2 [{shapeType=Capsule, length=2.0099, radius=0.8908, shapeName=dyn_medal2_capsule, center=<1.7897 0.0913 0.0286>, rotation=<rot 90 0 0>}]
 dyn_medal3 [{shapeType=Capsule, length=2.1462, radius=1.1808, shapeName=dyn_medal3_capsule, center=<2.1192 0.0012 0.0031>, rotation=<rot 90 0 0>}]
 dyn_medal4 [{shapeType=Capsule, length=1.9305, radius=1.0356, shapeName=dyn_medal4_capsule, center=<1.7714 -0.1061 -0.0155>, rotation=<rot 90 0 0>}]
 dyn_medal5 [{shapeType=Capsule, length=1.6041, radius=1.4683, shapeName=dyn_medal5_capsule, center=<0.9694 -0.2577 0.0875>, rotation=<rot 90 0 0>}]
 joint1 [{shapeType=Box, extentX=1, extentY=7, extentZ=7, shapeName=joint1_box, center=<0.4948 -0.8444 -4.1444>, rotation=<rot 0 30.3706 0>}]
 pack [{shapeType=Capsule, length=1, radius=1}]
 root [
  {shapeType=Sphere, radius=5.25, shapeName=root_sphere, center=<-5.3274 -10.6029 -1.8984>}
  {shapeType=Box, extentX=30, extentY=3.5, extentZ=35.5, shapeName=root_box, center=<-10.5626 -0.1255 -3.5178>}
 ]
}
constraints [
 {bone1Name=dyn_medal1, bone2Name=pack, swing1LimitDegrees=5, swing2LimitDegrees=5, twistLimitDegrees=5}
 {bone1Name=dyn_medal2, bone2Name=dyn_medal1, swing1LimitDegrees=15, swing2LimitDegrees=15, twistLimitDegrees=5}
 {bone1Name=dyn_medal3, bone2Name=dyn_medal2, swing1LimitDegrees=15, swing2LimitDegrees=15, twistLimitDegrees=5}
 {bone1Name=dyn_medal4, bone2Name=dyn_medal3, swing1LimitDegrees=15, swing2LimitDegrees=15, twistLimitDegrees=5}
 {bone1Name=dyn_medal5, bone2Name=dyn_medal4, swing1LimitDegrees=15, swing2LimitDegrees=15, twistLimitDegrees=5}
 {bone1Name=dyn_map_bot, bone2Name=dyn_map_base, swing1LimitDegrees=5, swing2LimitDegrees=5, twistLimitDegrees=5}
 {bone1Name=dyn_map_bot_end, bone2Name=dyn_map_bot, swing1LimitDegrees=5, swing2LimitDegrees=5, twistLimitDegrees=5}
 {bone1Name=dyn_map_top, bone2Name=dyn_map_base, swing1LimitDegrees=5, swing2LimitDegrees=5, twistLimitDegrees=5}
 {bone1Name=dyn_map_top_end, bone2Name=dyn_map_top, swing1LimitDegrees=5, swing2LimitDegrees=5, twistLimitDegrees=5}
 {
  bone1Name=dyn_map_base
  bone2Name=pack
  swing1LimitDegrees=45
  swing1Motion=Free
  swing2LimitDegrees=45
  swing2Motion=Free
  twistLimitDegrees=45
  twistMotion=Free
 }
 {bone1Name=dyn_flashlight, bone2Name=pack, swing1LimitDegrees=10, swing2LimitDegrees=10, twistLimitDegrees=10}
 {bone1Name=joint1, bone2Name=dyn_medal5, swing1LimitDegrees=5, swing2LimitDegrees=5, twistLimitDegrees=5}
 {
  bone1Name=pack
  bone2Name=root
  swing1LimitDegrees=45
  swing1Motion=Free
  swing2LimitDegrees=45
  swing2Motion=Free
  twistLimitDegrees=45
  twistMotion=Free
 }
]
